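; AssetPlacer is a small PCG example graph that marks an actor's location with a chosen asset, without spawning anything itself.
; Get Actor Data produces a single point, and Add Attribute writes the PickedAsset parameter onto it as an "asset" attribute.
; PickedAsset is a soft object path defaulting to an engine cube, and the tagged point leaves through the graph output.
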(graph AssetPlacer
  (params (PickedAsset :type (Struct /Script/CoreUObject.SoftObjectPath) :default (struct :AssetPath (struct :PackageName /Engine/BasicShapes/Cube1 :AssetName Cube1))))
  (bind GetActorData_0 (Get%20Actor%20Data :Mode GetSinglePoint))
  (bind AddAttribute_0 (Add%20Attribute :OutputTarget asset :Attributes PickedAsset))
  (connect GetActorData_0 Out AddAttribute_0 In)
  (output Out AddAttribute_0))
