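; PCG example graph that builds multi-floor building walls along tagged actor splines, fitted around hand-placed assets.
; Nearby HandPlacedAsset-tagged actors snap to spline samples, take rotation by projection with scale reset, then spawn.
; Spawned points become polygons that cut the splines into length-filtered segments for shape-grammar wall spawning.
; An upper-floor pass, with branches on floor and cutter counts, offsets splines by Base_Height and sizes walls from Floors.
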(graph Building_Wall_with_hand_Placed_Asset
  (params (ActorTag :type String :default "Building") (Floors :type Int :default 3) (Spawn_Blueprint :type Bool) (Base_Height :type Double :default 3.5e+02))
  (bind GetActorData (Subgraph :Graph /PCGPrimitives/Primitives/Get/Get_ActorData.Get_ActorData :ActorTag ActorTag))
  (bind FilterDataByType_11 (Filter%20Data%20By%20Type :TargetType (struct :Ids [(struct :Struct /Script/PCG.PCGDataTypeInfoPolyline)])))
  (bind SplineSampler_1 (Spline%20Sampler :SamplerParams (struct :Mode Distance :DistanceIncrement 5e+02 :bUnbounded true :PointSteepness 1.0)))
  (bind GetActorData_1 (Get%20Actor%20Data :ActorSelector (struct :ActorFilter AllWorldActors :ActorSelectionTag HandPlacedAsset :bSelectMultiple true) :bAlwaysRequeryActors true :bComponentsMustOverlapSelf false :bTrackActorsOnlyWithinBounds false :Mode GetDataFromPCGComponentOrParseComponents))
  (bind ToPoint_0 (To%20Point))
  (bind Distance_0 (Distance :bOutputDistanceVector true :bSetDensity true :MaximumDistance 1e+03 :SourceShape Center :TargetShape Center))
  (bind AttributeMathsOp_4 (Attribute%20Maths%20Op :InputSource1 $Position :InputSource2 Distance))
  (bind AttributeFilter_7 (Filter%20Attribute%20Elements :AttributeTypes (struct :Type Float :FloatValue 1.0 :DoubleValue 1.0) :bUseConstantThreshold true :Operator Lesser :ThresholdAttribute @Last)) ; if too far away dont 
  (bind MakeConcrete_1 (Make%20Concrete))
  (bind AttributeMathsOp_0 (Attribute%20Maths%20Op :InputSource1 $Scale :Operation Set))
  (bind MergePoints_4 (Merge%20Points))
  (bind Projection_0 (Projection :ProjectionParams (struct :bProjectPositions false :ColorBlendMode TargetValue))) ; get rotation
  (bind AttributeMathsOp_6 (Attribute%20Maths%20Op :InputSource1 $Scale :Operation Set)) ; killing scale as by default, dropped \nPCG volume have a 25,25,10 scale, messing the projection rotation.\nCurrently the system doesn't handle scaled assets.
  (bind Subgraph_21 (Subgraph :Graph /PCGPrimitives/Primitives/Spawn/Spawn_Assets.Spawn_Assets))
  (bind FilterDataByType_2 (Filter%20Data%20By%20Type :TargetType (struct :Ids [(struct :Struct /Script/PCG.PCGDataTypeInfoPoint)])))
  (bind MergePoints_1 (Merge%20Points :bMergeMetadata false))
  (bind AttributePartition_1 (Attribute%20Partition :PartitionAttributeSelectors [$Index]))
  (bind Loop_1 (Loop :Graph /PCGPrimitives/Primitives/Shared/convert_point_to_polygon.convert_point_to_polygon))
  (bind Subgraph_3 (Subgraph :Graph /PCGPrimitives/Primitives/Compose/Compose_Spline_Cut.Compose_Spline_Cut))
  (bind Subgraph_4 (Subgraph :Graph /PCGPrimitives/Primitives/Extract/Extract_Segments.Extract_Segments))
  (bind Subgraph_6 (Subgraph :Graph /PCGPrimitives/Primitives/Filter/Filter_Splines_byLength.Filter_Splines_byLength))
  (bind Subgraph_5 (Subgraph :Graph /PCGPrimitives/Primitives/Assign/Assign_ShapeGrammarDefinition.Assign_ShapeGrammarDefinition :Size Base_Height))
  (bind GetElementsCount_0 (Get%20Element%20Count))
  (bind AttributeReduce_1 (Attribute%20Reduce :Operation Sum))
  (bind AttributeCompareOp_1 (Attribute%20Compare%20Op :InA Floors))
  (bind Branch_0 (Branch))
  (bind Branch_1 (Branch))
  (bind Branch_2 (Branch))
  (bind MakeVectorAttribute_0 (Make%20Vector%20Attribute :OutputType Vector :Z Base_Height))
  (bind Subgraph_17 (Subgraph :Graph /PCGPrimitives/Primitives/Spawn/Spawn_Assets.Spawn_Assets))
  (bind Subgraph_0 (Subgraph :Graph /PCGPrimitives/Primitives/Debug/Debug_PrimitiveData.Debug_PrimitiveData))
  ;[0.15 0.15 0.15 0.5] UPPER FLOORS
  (bind Subgraph_7 (Subgraph :Graph /PCGPrimitives/Primitives/Compose/Compose_Spline_Cut.Compose_Spline_Cut))
  (bind AttributeMathsOp_3 (Attribute%20Maths%20Op :Operation Subtract :InA Floors))
  (bind AttributeMathsOp_1 (Attribute%20Maths%20Op :Operation Multiply :InB Base_Height))
  (bind Subgraph_14 (Subgraph :Graph /PCGPrimitives/Primitives/Transform/Transform_Offset.Transform_Offset))
  (bind Subgraph_8 (Subgraph :Graph /PCGPrimitives/Primitives/Extract/Extract_Segments.Extract_Segments))
  (bind Subgraph_10 (Subgraph :Graph /PCGPrimitives/Primitives/Filter/Filter_Splines_byLength.Filter_Splines_byLength))
  (bind Subgraph_9 (Subgraph :Graph /PCGPrimitives/Primitives/Assign/Assign_ShapeGrammarDefinition.Assign_ShapeGrammarDefinition))
  (reroute Input_Splines GetActorData)
  (connect AttributeMathsOp_4 Out AttributeFilter_7 In)
  (connect MakeConcrete_1 Out AttributeMathsOp_0 InA)
  (connect AttributeMathsOp_3 Out AttributeMathsOp_1 InA)
  (connect Distance_0 Out AttributeMathsOp_4 InA)
  (connect Distance_0 Out AttributeMathsOp_4 InB)
  (connect Projection_0 Out AttributeMathsOp_6 InA)
  (connect MergePoints_1 Out AttributePartition_1 In)
  (connect GetElementsCount_0 Out AttributeReduce_1 In)
  (connect AttributeCompareOp_1 Out Branch_0 bOutputToB)
  (connect Input_Splines Out Branch_0 In)
  (connect AttributeReduce_1 Out Branch_1 bOutputToB)
  (connect Branch_0 Output%20A Branch_1 In)
  (connect AttributeReduce_1 Out Branch_2 bOutputToB)
  (connect Loop_1 Out Branch_2 In)
  (connect ToPoint_0 Out Distance_0 Source)
  (connect SplineSampler_1 Out Distance_0 Target)
  (connect Subgraph_21 Out FilterDataByType_2 In)
  (connect Input_Splines Out FilterDataByType_11 In)
  (connect Loop_1 Out GetElementsCount_0 In)
  (connect AttributePartition_1 Out Loop_1 In)
  (connect AttributeFilter_7 InsideFilter MakeConcrete_1 In)
  (connect FilterDataByType_2 InsideFilter MergePoints_1 In)
  (connect SplineSampler_1 Out MergePoints_4 In)
  (connect AttributeMathsOp_0 Out Projection_0 In)
  (connect MergePoints_4 Out Projection_0 Projection%20Target)
  (connect FilterDataByType_11 InsideFilter SplineSampler_1 Spline)
  (connect Subgraph_10 Selected Subgraph_0 PrimaryInput)
  (connect Input_Splines Out Subgraph_3 PrimaryInput)
  (connect Loop_1 Out Subgraph_3 SecondaryInput)
  (connect Subgraph_3 Out Subgraph_4 PrimaryInput)
  (connect Subgraph_6 Selected Subgraph_5 PrimaryInput)
  (connect Subgraph_4 Out Subgraph_6 PrimaryInput)
  (connect Branch_1 Output%20B Subgraph_7 PrimaryInput)
  (connect Branch_2 Output%20B Subgraph_7 SecondaryInput)
  (connect Subgraph_14 Out Subgraph_8 PrimaryInput)
  (connect Subgraph_10 Selected Subgraph_9 PrimaryInput)
  (connect AttributeMathsOp_1 Out Subgraph_9 Size)
  (connect Subgraph_8 Out Subgraph_10 PrimaryInput)
  (connect MakeVectorAttribute_0 Out Subgraph_14 offset)
  (connect Subgraph_3 Out Subgraph_14 PrimaryInput)
  (connect Subgraph_7 Out Subgraph_14 PrimaryInput)
  (connect Subgraph_5 Points Subgraph_17 PrimaryInput)
  (connect Subgraph_9 Points Subgraph_17 PrimaryInput)
  (connect AttributeMathsOp_6 Out Subgraph_21 PrimaryInput)
  (connect GetActorData_1 Out ToPoint_0 In))
